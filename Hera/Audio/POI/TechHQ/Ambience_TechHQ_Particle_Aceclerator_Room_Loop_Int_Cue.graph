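# Looping indoor ambience cue for the TechHQ particle accelerator room, on the Athena indoor ambient sound class and attenuation.
# A quality node plays quad waves on its first two levels and stereo versions on the last two.
# Each branch mixes the room loop with borrowed temple chamber and lab computer-room beds, then chains looping envelopers for slow pitch drift and gentle volume swell.
# Concurrency allows one instance per owner, preventing new ones.
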SoundCue /Hera/Audio/POI/TechHQ/Ambience_TechHQ_Particle_Aceclerator_Room_Loop_Int_Cue
root=QualityLevel_0
nodes [
 {
  id=Enveloper_1
  children [Mixer_0]
  properties {
   bLoop=true
   bLoopIndefinitely=true
   loopEnd=30
   pitchCurve {editorCurveData {keys [{interpMode=RCIM_Cubic, time=0}, <at 15.9468 0.9181 RCIM_Cubic>, {interpMode=RCIM_Cubic, time=30}]}}
   volumeCurve {editorCurveData {keys [<at 0 1>]}}
  }
 }
 {
  id=Enveloper_2
  children [Enveloper_1]
  properties {
   bLoop=true
   bLoopIndefinitely=true
   loopEnd=20
   pitchCurve {editorCurveData {keys [{time=0}]}}
   volumeCurve {editorCurveData {keys [
      {arriveTangent=0, interpMode=RCIM_Cubic, leaveTangent=0, time=0, value=1}
      {arriveTangent=0, interpMode=RCIM_Cubic, leaveTangent=0, time=10, value=0.9}
      {arriveTangent=0, interpMode=RCIM_Cubic, leaveTangent=0, time=20, value=0.9992}
     ]}}
  }
 }
 {
  id=Enveloper_4
  children [Mixer_2]
  properties {
   bLoop=true
   bLoopIndefinitely=true
   loopEnd=30
   pitchCurve {editorCurveData {keys [{interpMode=RCIM_Cubic, time=0}, <at 15.9468 0.9181 RCIM_Cubic>, {interpMode=RCIM_Cubic, time=30}]}}
   volumeCurve {editorCurveData {keys [<at 0 1>]}}
  }
 }
 {
  id=Enveloper_5
  children [Enveloper_4]
  properties {
   bLoop=true
   bLoopIndefinitely=true
   loopEnd=20
   pitchCurve {editorCurveData {keys [{time=0}]}}
   volumeCurve {editorCurveData {keys [
      {arriveTangent=0, interpMode=RCIM_Cubic, leaveTangent=0, time=0, value=1}
      {arriveTangent=0, interpMode=RCIM_Cubic, leaveTangent=0, time=10, value=0.9}
      {arriveTangent=0, interpMode=RCIM_Cubic, leaveTangent=0, time=20, value=0.9992}
     ]}}
  }
 }
 {id=Mixer_0, children [WavePlayer_8,WavePlayer_0,WavePlayer_7], properties {inputVolume [2,1,0.6]}}
 {id=Mixer_2, children [WavePlayer_9,WavePlayer_1,WavePlayer_6], properties {inputVolume [2,1,0.6]}}
 {id=QualityLevel_0, children [Enveloper_2,Enveloper_2,Enveloper_5,Enveloper_5]}
 {
  id=WavePlayer_0
  properties {bLooping=true, soundWaveAssetPtr=/AmbientAudioContent/Waves/POI/Ambience_AncientTemple_UndergroundChamber_EveningNight_Loop_Quad}
 }
 {
  id=WavePlayer_1
  properties {bLooping=true, soundWaveAssetPtr=/AmbientAudioContent/Waves/POI/Ambience_AncientTemple_UndergroundChamber_EveningNight_Loop_Stereo}
 }
 {
  id=WavePlayer_6
  properties {bLooping=true, soundWaveAssetPtr=/AmbientAudioContent/Waves/POI/Ambience_HydroBottle_Lab_ComputerRoom_Loop_Airy_Stereo}
 }
 {
  id=WavePlayer_7
  properties {bLooping=true, soundWaveAssetPtr=/AmbientAudioContent/Waves/POI/Ambience_HydroBottle_Lab_ComputerRoom_Loop_Airy_Quad}
 }
 {
  id=WavePlayer_8
  properties {bLooping=true, soundWaveAssetPtr=&/Ambience_TechHQ_Particle_Accelerator_Room_Loop_Int_Quad}
 }
 {
  id=WavePlayer_9
  properties {bLooping=true, soundWaveAssetPtr=&/Ambience_TechHQ_Particle_Accelerator_Room_Loop_Int_Stereo}
 }
]
properties {
 attenuationSettings=&Fort_Ambient_Athena_Att
 bOverrideConcurrency=true
 maxDistance=2097152
 soundClassObject=&Fort_Ambient_Indoors_Athena
 volumeMultiplier=0.3
 concurrencyOverrides {bLimitToOwner=true, maxCount=1, resolutionRule=PreventNew, retriggerTime=2}
}
