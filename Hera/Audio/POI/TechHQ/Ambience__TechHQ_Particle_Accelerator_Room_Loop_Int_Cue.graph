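# Looping indoor ambience cue for the TechHQ particle accelerator room, on an indoor ambient sound class with single-instance, prevent-new concurrency.
# A quality-level root sends its first two slots to a quad branch and its last two to a stereo one.
# Each branch mixes room, temple chamber and computer room waves, the last not flagged looping.
# Nested, endlessly looping envelopers add slow pitch modulation and a gentle volume swell.
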SoundCue /Hera/Audio/POI/TechHQ/Ambience__TechHQ_Particle_Accelerator_Room_Loop_Int_Cue
root=QualityLevel_0
nodes [
 {
  id=Enveloper_0
  children [Mixer_3]
  properties {
   bLoop=true
   bLoopIndefinitely=true
   loopEnd=30
   pitchCurve {editorCurveData {keys [{interpMode=RCIM_Cubic, time=0}, <at 15.9468 0.9181 RCIM_Cubic>, {interpMode=RCIM_Cubic, time=30}]}}
   volumeCurve {editorCurveData {keys [<at 0 1>]}}
  }
 }
 {
  id=Enveloper_1
  children [Mixer_2]
  properties {
   bLoop=true
   bLoopIndefinitely=true
   loopEnd=30
   pitchCurve {editorCurveData {keys [{interpMode=RCIM_Cubic, time=0}, <at 15.9468 0.9181 RCIM_Cubic>, {interpMode=RCIM_Cubic, time=30}]}}
   volumeCurve {editorCurveData {keys [<at 0 1>]}}
  }
 }
 {
  id=Enveloper_2
  children [Enveloper_1]
  properties {
   bLoop=true
   bLoopIndefinitely=true
   loopEnd=20
   pitchCurve {editorCurveData {keys [{time=0}]}}
   volumeCurve {editorCurveData {keys [
      {arriveTangent=-0.01, interpMode=RCIM_Cubic, time=0, value=1}
      <at 10 1.1 RCIM_Cubic>
      {interpMode=RCIM_Cubic, leaveTangent=0.01, time=20, value=1}
     ]}}
  }
 }
 {
  id=Enveloper_3
  children [Enveloper_0]
  properties {
   bLoop=true
   bLoopIndefinitely=true
   loopEnd=30
   pitchCurve {editorCurveData {keys [{time=0}]}}
   volumeCurve {editorCurveData {keys [
      {arriveTangent=-0.01, interpMode=RCIM_Cubic, time=0, value=1}
      <at 10 1.1 RCIM_Cubic>
      {interpMode=RCIM_Cubic, leaveTangent=0.01, time=20, value=1}
     ]}}
  }
 }
 {id=Mixer_2, children [WavePlayer_4, WavePlayer_6, WavePlayer_12], properties {inputVolume [1.5,1,0.66]}}
 {id=Mixer_3, children [WavePlayer_5, WavePlayer_7, WavePlayer_13], properties {inputVolume [1.5,1,0.66]}}
 {id=QualityLevel_0, children [Enveloper_2,Enveloper_2,Enveloper_3,Enveloper_3]}
 {id=WavePlayer_12, properties {soundWaveAssetPtr=/AmbientAudioContent/Waves/POI/Ambience_HydroBottle_Lab_ComputerRoom_Loop_Airy_Quad}}
 {id=WavePlayer_13, properties {soundWaveAssetPtr=/AmbientAudioContent/Waves/POI/Ambience_HydroBottle_Lab_ComputerRoom_Loop_Airy_Stereo}}
 {
  id=WavePlayer_4
  properties {bLooping=true, soundWaveAssetPtr=&/Ambience_TechHQ_Particle_Accelerator_Room_Loop_Int_Quad}
 }
 {
  id=WavePlayer_5
  properties {bLooping=true, soundWaveAssetPtr=&/Ambience_TechHQ_Particle_Accelerator_Room_Loop_Int_Stereo}
 }
 {
  id=WavePlayer_6
  properties {bLooping=true, soundWaveAssetPtr=/AmbientAudioContent/Waves/POI/Ambience_AncientTemple_UndergroundChamber_EveningNight_Loop_Quad}
 }
 {
  id=WavePlayer_7
  properties {bLooping=true, soundWaveAssetPtr=/AmbientAudioContent/Waves/POI/Ambience_AncientTemple_UndergroundChamber_EveningNight_Loop_Stereo}
 }
]
properties {
 attenuationSettings=/AmbientAudioBase/Attenuations/Fort_Ambient_Att
 bEnableBusSends=false
 bOverrideConcurrency=true
 maxDistance=2097152
 soundClassObject=&Fort_Ambient_Indoors_Athena
 volumeMultiplier=0.5
 concurrencyOverrides {maxCount=1, resolutionRule=PreventNew}
 concurrencySet [/AmbientAudioBase/Concurrency/AmbientConc_2Max_Loop_PreventNew_Default_Concurrency]
 preEffectBusSends [{}]
}
