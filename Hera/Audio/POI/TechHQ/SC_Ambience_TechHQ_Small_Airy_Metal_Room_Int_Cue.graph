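# Quiet, non-spatialized indoor ambience cue for a small, airy metal room in the TechHQ POI.
# A quality-level switch sends its first two levels to a quad room-tone mix and its last two to a stereo one.
# Each mix blends residential room tone with a pitched-down looping air-duct bed.
# Nested looping envelopers of different lengths slowly vary the room tone's pitch and volume.
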SoundCue /Hera/Audio/POI/TechHQ/SC_Ambience_TechHQ_Small_Airy_Metal_Room_Int_Cue
root=QualityLevel_0
nodes [
 {
  id=Enveloper_0
  children [WavePlayer_0]
  properties {
   bLoop=true
   bLoopIndefinitely=true
   loopEnd=45
   pitchCurve {editorCurveData {keys [{interpMode=RCIM_Cubic, time=0}, <at 22.5 0.91 RCIM_Cubic>, {interpMode=RCIM_Cubic, time=45}]}}
   volumeCurve {editorCurveData {keys [<at 0 1>]}}
  }
 }
 {
  id=Enveloper_3
  children [Enveloper_0]
  properties {
   bLoop=true
   bLoopIndefinitely=true
   loopEnd=20
   pitchCurve {editorCurveData {keys [{time=0}]}}
   volumeCurve {editorCurveData {keys [
      {arriveTangent=-0.01, interpMode=RCIM_Cubic, time=0, value=1}
      <at 10 0.95 RCIM_Cubic>
      {interpMode=RCIM_Cubic, leaveTangent=0.01, time=20, value=1}
     ]}}
  }
 }
 {
  id=Enveloper_4
  children [WavePlayer_2]
  properties {
   bLoop=true
   bLoopIndefinitely=true
   loopEnd=45
   pitchCurve {editorCurveData {keys [{interpMode=RCIM_Cubic, time=0}, <at 22.5 0.91 RCIM_Cubic>, {interpMode=RCIM_Cubic, time=45}]}}
   volumeCurve {editorCurveData {keys [<at 0 1>]}}
  }
 }
 {
  id=Enveloper_5
  children [Enveloper_4]
  properties {
   bLoop=true
   bLoopIndefinitely=true
   loopEnd=20
   pitchCurve {editorCurveData {keys [{time=0}]}}
   volumeCurve {editorCurveData {keys [
      {arriveTangent=-0.01, interpMode=RCIM_Cubic, time=0, value=1}
      <at 10 0.95 RCIM_Cubic>
      {interpMode=RCIM_Cubic, leaveTangent=0.01, time=20, value=1}
     ]}}
  }
 }
 {id=Mixer_0, children [Enveloper_3,Modulator_2], properties {inputVolume [1.2,1]}}
 {id=Mixer_1, children [Enveloper_5,Modulator_3], properties {inputVolume [1.2,1]}}
 {id=Modulator_2, children [WavePlayer_5], properties {pitchMax=0.8, pitchMin=0.8}}
 {id=Modulator_3, children [WavePlayer_6], properties {pitchMax=0.8, pitchMin=0.8}}
 {id=QualityLevel_0, children [Mixer_0,Mixer_0,Mixer_1,Mixer_1]}
 {id=WavePlayer_0, properties {soundWaveAssetPtr=/AmbientAudioContent/Waves/POI/Ambience_Residential_Medium_Room_Tone_Quad}}
 {id=WavePlayer_2, properties {soundWaveAssetPtr=/AmbientAudioContent/Waves/POI/Ambience_Residential_Medium_Room_Tone_Stereo}}
 {id=WavePlayer_5, properties {bLooping=true, soundWaveAssetPtr=/AmbientAudioContent/Waves/SpottedLoops/Spot_AMB_Air_Duct_Interior_Loop_01}}
 {id=WavePlayer_6, properties {bLooping=true, soundWaveAssetPtr=&Spot_AMB_Air_Duct_Interior_Loop_01}}
]
properties {
 bEnableBusSends=false
 bEnableSubmixSends=false
 bOverrideConcurrency=true
 maxDistance=2097152
 soundClassObject=&Fort_Ambient_Indoors_Athena
 volumeMultiplier=0.2
 attenuationOverrides {bAttenuate=false, bEnableReverbSend=false, bSpatialize=false}
 concurrencyOverrides {resolutionRule=PreventNew}
}
